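# Animation Blueprint for the Badass Cowboy Plunger backpack cosmetic, adding procedural secondary motion to its handle and bomb bones.
# The local-space reference pose feeds two chained Anim Dynamics nodes in component space, first dyn_handle, then dyn_bomb, before converting back for output.
# Both use damped linear springs, gravity and component-motion response, with tight translation limits and zero angular limits, so the parts slide rather than swing.
# The event graph has no logic.
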AnimBlueprint /BRCosmetics/Accessories/FORT_Backpacks/Backpack_Badass_Cowboy_Plunger/Meshes/M_MED_Badass_Cowboy_Plunger_AnimBlueprint
target_skeleton=&/M_MED_Badass_Cowboy_Plunger_Skeleton
graphs [
 AnimGraph {nodes [
   n1 = Anim Dynamics - Bone: dyn_handle(ComponentPose=Local To Component(LocalPose=Local Space Ref Pose()), Alpha=1, ExternalForce="0,0,0") {
    bGravityOverrideInSimSpace=true
    bLinearSpring=true
    bOverrideLinearDamping=true
    bUseGravityOverride=true
    linearDampingOverride=0.998
    linearSpringConstant=200
    alphaBoolBlend {blendOption=Linear}
    boundBone {boneName=dyn_handle}
    componentAppliedLinearAccClamp=<3000 3000 3000>
    componentLinearAccScale=<0.3 0.3 0.5>
    componentLinearVelScale=<0.3 0.3 0.5>
    gravityOverride {z=-980}
    physicsBodyDefinitions [{
      sphereCollisionRadius=0
      boundBone {boneName=dyn_handle}
      boxExtents=<10 5 5>
      constraintSetup {angularLimitsMax=<0 0 0>, angularLimitsMin=<0 0 0>, angularTarget=<0 0 0>, linearAxesMax=<0 0 2>, linearAxesMin=<0 0 -2>}
      localJointOffset=<0 0 -1>
     }]
   }
   n2 = Anim Dynamics - Bone: dyn_bomb(ComponentPose=@n1, Alpha=1, ExternalForce="0,0,0") {
    bGravityOverrideInSimSpace=true
    bLinearSpring=true
    bOverrideLinearDamping=true
    bUseGravityOverride=true
    bUsePlanarLimit=false
    linearDampingOverride=1
    linearSpringConstant=100
    alphaBoolBlend {blendOption=Linear}
    boundBone {boneName=dyn_bomb}
    componentLinearAccScale=<1 1 1>
    componentLinearVelScale=<1 1 1>
    gravityOverride {z=-980}
    physicsBodyDefinitions [{
      sphereCollisionRadius=0
      boundBone {boneName=dyn_bomb}
      boxExtents=<5 20 20>
      constraintSetup {angularLimitsMax=<0 0 0>, angularLimitsMin=<0 0 0>, angularTarget=<0 0 0>, linearAxesMax=<2.5 0 0>, linearAxesMin=<-2.5 -2 -0.5>}
      localJointOffset=<0 0.1 -0.1>
     }]
   }
   Output Pose(Result=Component To Local(ComponentPose=@n2)) {layerGroup=None}
  ]}
 EventGraph {nodes [Event BlueprintUpdateAnimation()]}
]
variables []
